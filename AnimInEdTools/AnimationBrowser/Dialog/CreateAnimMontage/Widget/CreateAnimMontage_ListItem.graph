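# Editor utility list-entry widget for the Create AnimMontage dialog in the Animation Browser tool, showing one item name in a bordered text row.
# When its list item is set, it casts it to CreateAnimMontage_Object, displays its name and clears the highlight.
# Selection changes drive Display Selected, which toggles a blue border and near-black text against the original text color cached on Construct.
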EditorUtilityWidgetBlueprint /AnimInEdTools/AnimationBrowser/Dialog/CreateAnimMontage/Widget/CreateAnimMontage_ListItem
graphs [
 Display Selected {nodes [
   n1 = Display Selected!() -> (then, Index:bool=~) {is=title K2Node_FunctionEntry}
   > SetBrushColor(
    self=Get Border() -> (Border:Border Object Reference)
    InBrushColor=SelectColor(A=(R=0,G=0.1608,B=0.7451,A=1), B=(R=0,G=0,B=0,A=0), bPickA=@n1)
   )
   > SetColorAndOpacity(
    self=Get TextBlock() -> (TextBlock:Text Object Reference)
    InColorAndOpacity=Select(
     Option 0:Slate Color Structure=Get Color And Opacity() -> (Color And Opacity:Slate Color Structure)
     Option 1:Slate Color Structure=Make Slate Color(SpecifiedColor=(R=0.0103,G=0.0103,B=0.0103,A=1))
     Index=@n1
    ) -> (ReturnValue:Slate Color Structure)
   )
  ]}
 EventGraph {nodes [
   n1 = Event OnListItemObjectSet()
   > n2 = Cast To CreateAnimMontage_Object!(execute, Object=@n1.ListItemObject) -> (then, CastFailed, AsCreate Anim Montage Object:Create Anim Montage Object Object Reference) {is=title K2Node_DynamicCast}
   > Set Text(Text="To Text (String)"(InString=Get Name!(self:Create Anim Montage Object Object Reference=@n2) -> (Name:String)), self=Get TextBlock())
   > Display Selected!(execute, self, Index:bool=false) -> (then) {is=title K2Node_CallFunction}
   Event Construct()
   > Set Color And Opacity(Color And Opacity:Slate Color Structure=Get ColorAndOpacity!(self:Text Object Reference=Get TextBlock()) -> (ColorAndOpacity:Slate Color Structure)) -> (Output_Get:Slate Color Structure)
   n3 = Event On Item Selection Changed!() -> (OutputDelegate, then, bIsSelected:bool) {is=title K2Node_Event}
   > Display Selected(Index=@n3.bIsSelected)
   Event On Item Expansion Changed!() -> (OutputDelegate, then, bIsExpanded:bool) {is=title K2Node_Event}
   Event On Entry Released() {is=title K2Node_Event}
   Reroute Node(InputPin:String=Reroute Node()) -> (OutputPin:String)
  ]}
]
variables [Color And Opacity {type=SlateColor, default {colorUseRule=UseColor_Specified, specifiedColor=<rgba 1 0 1 1>}}]
widgets {
 info {rootWidgetClass=&Border, widgetCount=3}
 widgets [
  Border {widgetClassPath=&Border, bIsVariable=true, widget=Border, properties {brushColor=<rgba 1 1 1 -969.0196>}}
  HorizontalBox_0 {widgetClassPath=&HorizontalBox, parent=Border, widget=HorizontalBox_0, slot {class=BorderSlot}}
  TextBlock {
   widgetClassPath=&TextBlock
   bIsVariable=true
   parent=HorizontalBox_0
   widget=TextBlock
   properties {
    text=Binding
    colorAndOpacity {colorUseRule=UseColor_Specified, specifiedColor=<rgba 0.5255 0.5255 0.5255 1>}
    font {
     bForceMonospaced=false
     bMaterialIsStencil=false
     fontMaterial=null
     fontObject=&Roboto
     letterSpacing=0
     monospacedWidth=1
     size=10
     skewAmount=0
     typefaceFontName=Regular
     outlineSettings {
      bApplyOutlineToDropShadows=false
      bMiteredCorners=false
      bSeparateFillAlpha=false
      outlineMaterial=null
      outlineSize=0
      outlineColor=<rgba 0 0 0 1>
     }
    }
   }
   slot {class=HorizontalBoxSlot, properties {padding=<ltrb 4 2 4 2>, size {sizeRule=Fill, value=1}}}
  }
 ]
}
